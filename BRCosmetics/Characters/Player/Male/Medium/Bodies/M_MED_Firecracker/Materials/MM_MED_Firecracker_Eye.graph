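# Masked character master material for the eye of the Firecracker player body, built on the shared character function chain.
# PreFX and CharacterSkinFX supply the attributes, which are edited before passing through PostFX and the cinematic post function.
# An eye mask texture's alpha times Show Eye drives opacity mask, while eyebrow texture alpha times ExplodedEyeColorMultiplier is added to base color and saturated.
# Both scalars are noted as triggered by CID variants.
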Material /BRCosmetics/Characters/Player/Male/Medium/Bodies/M_MED_Firecracker/Materials/MM_MED_Firecracker_Eye
graph {
 properties {
  bAutomaticallySetUsageInEditor=false
  bUseMaterialAttributes=true
  bUsedWithClothing=true
  bUsedWithMeshDeformer=true
  bUsedWithMorphTargets=true
  bUsedWithSkeletalMesh=true
  blendMode=BLEND_Masked
  pixelDepthOffsetMode=PDOM_Legacy
  refractionMethod=RM_IndexOfRefraction
  shadingModel=MSM_FromMaterialExpression
  subsurfaceProfile=&SS_HeroSkin_02
  usedShadingModels=Default Lit | Subsurface | Clear Coat | Subsurface Profile | Cloth | Eye
 }
 nodes [
  n1 = MaterialFunctionCall() {materialFunction=&BaseFN_CharacterFunction_PreFX}
  n2 = GetMaterialAttributes(None=MaterialFunctionCall(
    PreskinnedPosition=@n1.Pre skinned Local Position
    Gmap=@n1.Pre skinned Local Position
    MaterialAttributes=@n1.Pre skinned Local Position
    AO=@n1.Pre skinned Local Position
   ) {materialFunction=&CharacterSkinFX}.Result) {attributeGetTypes [#1, #2]}
  n3 = TextureSampleParameter2D() {
   parameterName=Firecracker Eye Mask
   texture=/BRCosmetics/Characters/Player/Male/Medium/Bodies/M_MED_Firecracker/Textures/T_M_MED_Firecracker_Head_Flipbook_Eyes_Eyebrow
  }
  n4 = ScalarParameter() {defaultValue=1, desc=Triggered in CID Variant, group=Firecracker, parameterName=Show Eye, sliderMax=1}
  n5 = Add(
   A=@n2.Base Color
   B=Multiply(
    A=TextureSample() {texture=/BRCosmetics/Characters/Player/Male/Medium/Bodies/M_MED_Firecracker/Textures/T_M_MED_Firecracker_Head_Flipbook_Eyebrow}.A
    B=ScalarParameter() {desc=Triggered in CID variant, group=Firecracker, parameterName=ExplodedEyeColorMultiplier}
   ) {constB=0.5}
  )
  n6 = MaterialFunctionCall(
   AO=@n1.AO
   Distance Blend=@n1.AO
   Fresnel=@n1.AO
   Gmap=@n1.AO
   Pre Skinned Local Position=@n1.AO
   Material=SetMaterialAttributes!(MaterialAttributes=@n2.MaterialAttributes, Opacity Mask=Multiply(A=@n3.A, B=@n4), Base Color=Saturate(None=@n5)) {attributeSetTypes [#3, #1]}
  ) {materialFunction=&BaseFN_CharacterFunction_PostFX}
  MaterialFunctionCall(In=@n6.Result) {materialFunction=&CINE_MF_CharPostFX}
 ]
}
