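# Master material for the mouth of the medium male Firecracker character, built on shared character functions for masked, skinned meshes.
# PreFX data feeds CharacterSkinFX, whose base color is raised to the tunable "Base Color Power Adjust" exponent before the character PostFX and cinematic PostFX functions.
# Shading model comes from the expression graph, with a hero skin subsurface profile assigned.
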Material /BRCosmetics/Characters/Player/Male/Medium/Bodies/M_MED_Firecracker/Materials/MM_MED_Firecracker_Mouth
graph {
 properties {
  bAutomaticallySetUsageInEditor=false
  bUseMaterialAttributes=true
  bUsedWithClothing=true
  bUsedWithMeshDeformer=true
  bUsedWithMorphTargets=true
  bUsedWithSkeletalMesh=true
  blendMode=BLEND_Masked
  pixelDepthOffsetMode=PDOM_Legacy
  refractionMethod=RM_IndexOfRefraction
  shadingModel=MSM_FromMaterialExpression
  subsurfaceProfile=&SS_HeroSkin_02
  usedShadingModels=Default Lit | Subsurface | Clear Coat | Subsurface Profile | Cloth | Eye
 }
 nodes [
  n1 = MaterialFunctionCall() {materialFunction=&BaseFN_CharacterFunction_PreFX}
  n2 = GetMaterialAttributes(None=MaterialFunctionCall(
    PreskinnedPosition=@n1.Pre skinned Local Position
    Gmap=@n1.Pre skinned Local Position
    MaterialAttributes=@n1.Pre skinned Local Position
    AO=@n1.Pre skinned Local Position
   ) {materialFunction=&CharacterSkinFX}.Result) {attributeGetTypes [#1]}
  n3 = ScalarParameter() {defaultValue=3, group=Firecracker, parameterName=Base Color Power Adjust}
  n4 = MaterialFunctionCall(
   AO=@n1.AO
   Distance Blend=@n1.AO
   Fresnel=@n1.AO
   Gmap=@n1.AO
   Pre Skinned Local Position=@n1.AO
   Material=SetMaterialAttributes!(MaterialAttributes=@n2.MaterialAttributes, Base Color=Power(Base=@n2.Base Color, Exp=@n3) {constExponent=3}) {attributeSetTypes [#1]}
  ) {materialFunction=&BaseFN_CharacterFunction_PostFX}
  MaterialFunctionCall(In=@n4.Result) {materialFunction=&CINE_MF_CharPostFX}
 ]
}
