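# Physics asset for the Firecracker medium male player body, covering the character's dynamic wick bone chain.
# A large box body on dyn_wick_mstr forms the root, followed by three small capsule bodies along wick bones 01 to 03.
# Constraints link each segment to the one before it, with swing and twist limits generally loosening toward the tip so the end bends more freely than the base.
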PhysicsAsset /Game/Characters/Player/Male/Medium/Bodies/M_MED_Firecracker/Meshes/Parts/M_MED_Firecracker_Physics
bodies {
 dyn_wick_01 [{shapeType=Capsule, length=3.9602, radius=2.7436, center=<3.3149 0 0>, rotation=<rot 90 0 0>}]
 dyn_wick_02 [{shapeType=Capsule, length=3.9366, radius=2.3175, center=<3 0 0>, rotation=<rot 90 0 0>}]
 dyn_wick_03 [{shapeType=Capsule, length=3.7221, radius=2.2972, center=<4 0 0>, rotation=<rot 90 0 0>}]
 dyn_wick_mstr [{shapeType=Box, extentX=8.5923, extentY=32.3576, extentZ=32.1944, center=<-1.13 -4.9715 3.8313>, rotation=<rot 22.7139 0 0>}]
}
constraints [
 {bone1Name=dyn_wick_03, bone2Name=dyn_wick_02, swing1LimitDegrees=45, swing2LimitDegrees=45, twistLimitDegrees=15}
 {bone1Name=dyn_wick_01, bone2Name=dyn_wick_mstr, swing1LimitDegrees=25, swing2LimitDegrees=35, twistLimitDegrees=5}
 {bone1Name=dyn_wick_02, bone2Name=dyn_wick_01, swing1LimitDegrees=35, swing2LimitDegrees=45, twistLimitDegrees=10}
]
