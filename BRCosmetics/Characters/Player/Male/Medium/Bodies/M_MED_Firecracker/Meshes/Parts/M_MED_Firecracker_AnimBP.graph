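# Post-copy AnimBlueprint for the M_MED_Firecracker player body part, adding secondary motion to its eyepatch and wick bones.
# It copies the attached parent's pose and curves, constrains the eyepatch master bones mostly to neck_02, then simulates eyepatch and wick bones with chained AnimDynamics.
# An alternate wick chain and a RigidBody node are left unconnected to the output.
# Each update switches the material's smile mouth frame based on whether an emote is playing.
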AnimBlueprint /BRCosmetics/Characters/Player/Male/Medium/Bodies/M_MED_Firecracker/Meshes/Parts/M_MED_Firecracker_AnimBP
target_skeleton=/Game/Characters/Player/Male/Medium/Bodies/M_MED_Firecracker/Meshes/Parts/M_MED_Firecracker_Skeleton
graphs [
 AnimGraph {nodes [
   n1 = Anim Dynamics - Bone: dyn_wick_01(Alpha=1, ExternalForce="0,0,0") {
    angularDampingOverride=0.98
    angularSpringConstant=150
    bAngularSpring=true
    bGravityOverrideInSimSpace=true
    bOverrideAngularDamping=true
    simulationSpace=BoneRelative
    alphaBoolBlend {blendOption=Linear}
    boundBone {boneName=dyn_wick_01}
    componentLinearAccScale=<1.25 1.25 1.25>
    componentLinearVelScale=<1.25 1.25 1.25>
    physicsBodyDefinitions [{
      sphereCollisionRadius=0
      boundBone {boneName=dyn_wick_01}
      boxExtents=<15 10 10>
      constraintSetup {angularLimitsMax=<5 0 5>, angularLimitsMin=<-5 -25 -5>, angularTarget=<0 0 0>}
      localJointOffset=<1.5 0 0>
     }]
    relativeSpaceBone {boneName=root}
   }
   n2 = Anim Dynamics - Bone: dyn_wick_02(ComponentPose=@n1, Alpha=1, ExternalForce="0,0,0") {
    angularDampingOverride=0.98
    angularSpringConstant=150
    bAngularSpring=true
    bGravityOverrideInSimSpace=true
    bOverrideAngularDamping=true
    simulationSpace=BoneRelative
    alphaBoolBlend {blendOption=Linear}
    boundBone {boneName=dyn_wick_02}
    componentLinearAccScale=<1.25 1.25 1.25>
    componentLinearVelScale=<1.25 1.25 1.25>
    physicsBodyDefinitions [{
      sphereCollisionRadius=0
      boundBone {boneName=dyn_wick_02}
      boxExtents=<15 10 10>
      constraintSetup {angularLimitsMax=<5 0 5>, angularLimitsMin=<-5 -25 -5>, angularTarget=<0 0 0>}
      localJointOffset=<1.5 0 0>
     }]
    relativeSpaceBone {boneName=root}
   }
   Anim Dynamics - Bone: dyn_wick_03(ComponentPose=@n2, Alpha=1, ExternalForce="0,0,0") {
    angularDampingOverride=0.98
    angularSpringConstant=150
    bAngularSpring=true
    bGravityOverrideInSimSpace=true
    bOverrideAngularDamping=true
    simulationSpace=BoneRelative
    alphaBoolBlend {blendOption=Linear}
    boundBone {boneName=dyn_wick_03}
    componentLinearAccScale=<1.25 1.25 1.25>
    componentLinearVelScale=<1.25 1.25 1.25>
    physicsBodyDefinitions [{
      sphereCollisionRadius=0
      boundBone {boneName=dyn_wick_03}
      boxExtents=<15 10 10>
      constraintSetup {angularLimitsMax=<5 0 5>, angularLimitsMin=<-5 -25 -5>, angularTarget=<0 0 0>}
      localJointOffset=<1.5 0 0>
     }]
    relativeSpaceBone {boneName=root}
   }
   RigidBody() {
    bClampLinearTranslationLimitToRefPose=true
    bOverrideWorldGravity=true
    overridePhysicsAsset=/Game/Characters/Player/Male/Medium/Bodies/M_MED_Firecracker/Meshes/Parts/M_MED_Firecracker_Physics
    simulationSpace=BaseBoneSpace
    alphaBoolBlend {blendOption=Linear}
    baseBoneRef {boneName=neck_01}
    componentAppliedLinearAccClamp=<3000 3000 3000>
    componentLinearAccScale=<0.3 0.4 0.4>
    componentLinearVelScale=<0.2 0.25 0.25>
    overrideWorldGravity {z=-540}
   }
   n3 = "Constraint - dyn_eyepatch_01_MSTR (PPPPPPP)"!(
    ComponentPose=Local To Component(LocalPose=Copy Pose From Mesh() {bCopyCurves=true, bUseAttachedParent=true})
    Alpha=1
    ConstraintWeights_0=0
    ConstraintWeights_1:float=0
    ConstraintWeights_2:float=0.0001
    ConstraintWeights_3:float=0.0001
    ConstraintWeights_4:float=0.0003
    ConstraintWeights_5:float=0.0007
    ConstraintWeights_6:float=0.9988
   ) -> (Pose) {
    alphaBoolBlend {blendOption=Linear}
    boneToModify {boneName=dyn_eyepatch_01_MSTR}
    constraintSetup [
     {targetBone {boneName=clavicle_l}}
     {targetBone {boneName=spine_02}}
     {targetBone {boneName=spine_03}}
     {targetBone {boneName=spine_04}}
     {targetBone {boneName=spine_05}}
     {targetBone {boneName=clavicle_r}}
     {targetBone {boneName=neck_02}}
    ]
    constraintWeights [1,1,1,1,1,1,1]
   }
   n4 = "Constraint - dyn_eyepatch_02_MSTR (PPPPPPP)"!(
    ComponentPose=@n3
    Alpha=1
    ConstraintWeights_0=0
    ConstraintWeights_1:float=0
    ConstraintWeights_2:float=0.0001
    ConstraintWeights_3:float=0
    ConstraintWeights_4:float=0.0001
    ConstraintWeights_5:float=0.0003
    ConstraintWeights_6:float=0.9995
   ) -> (Pose) {
    alphaBoolBlend {blendOption=Linear}
    boneToModify {boneName=dyn_eyepatch_02_MSTR}
    constraintSetup [
     {targetBone {boneName=clavicle_l}}
     {targetBone {boneName=spine_02}}
     {targetBone {boneName=spine_03}}
     {targetBone {boneName=spine_04}}
     {targetBone {boneName=spine_05}}
     {targetBone {boneName=clavicle_r}}
     {targetBone {boneName=neck_02}}
    ]
    constraintWeights [1,1,1,1,1,1,1]
   }
   n5 = Anim Dynamics - Bone: dyn_eyepatch_01(ComponentPose=@n4, Alpha=1, ExternalForce="0,0,0") {
    angularDampingOverride=0.998
    angularSpringConstant=150
    bAngularSpring=true
    bGravityOverrideInSimSpace=true
    bOverrideAngularDamping=true
    bUseGravityOverride=true
    simulationSpace=World
    alphaBoolBlend {blendOption=Linear}
    boundBone {boneName=dyn_eyepatch_01}
    componentAppliedLinearAccClamp=<1500 1500 1500>
    componentLinearAccScale=<0.3 0.3 0.3>
    componentLinearVelScale=<0.3 0.3 0.3>
    gravityOverride {z=-980}
    physicsBodyDefinitions [{
      sphereCollisionRadius=0
      boundBone {boneName=dyn_eyepatch_01}
      boxExtents=<15 10 10>
      constraintSetup {angularLimitsMax=<15 15 25>, angularLimitsMin=<-15 -45 -25>, angularTarget=<0 0 0>}
      localJointOffset=<2 -1 0>
     }]
    relativeSpaceBone {boneName=spine_05}
   }
   n6 = Anim Dynamics - Bone: dyn_eyepatch_02(ComponentPose=@n5, Alpha=1, ExternalForce="0,0,0") {
    angularDampingOverride=0.998
    angularSpringConstant=150
    bAngularSpring=true
    bGravityOverrideInSimSpace=true
    bOverrideAngularDamping=true
    bUseGravityOverride=true
    simulationSpace=World
    alphaBoolBlend {blendOption=Linear}
    boundBone {boneName=dyn_eyepatch_02}
    componentAppliedLinearAccClamp=<1000 1000 1000>
    componentLinearAccScale=<0.3 0.3 0.3>
    componentLinearVelScale=<0.3 0.3 0.3>
    gravityOverride {z=-980}
    physicsBodyDefinitions [{
      sphereCollisionRadius=0
      boundBone {boneName=dyn_eyepatch_02}
      boxExtents=<15 10 10>
      constraintSetup {angularLimitsMax=<15 10 15>, angularLimitsMin=<-15 -45 -15>, angularTarget=<0 0 0>}
      localJointOffset=<2 1 0>
     }]
    relativeSpaceBone {boneName=spine_05}
   }
   n7 = Anim Dynamics - Bone: dyn_wick_01(ComponentPose=@n6, Alpha=1, ExternalForce="0,0,0") {
    angularDampingOverride=0.9999
    angularSpringConstant=50
    bAngularSpring=true
    bGravityOverrideInSimSpace=true
    bOverrideAngularDamping=true
    simulationSpace=RootRelative
    alphaBoolBlend {blendOption=Linear}
    boundBone {boneName=dyn_wick_01}
    componentAppliedLinearAccClamp=<3500 3500 3500>
    componentLinearAccScale=<3 3 3>
    componentLinearVelScale=<3 3 3>
    physicsBodyDefinitions [{
      sphereCollisionRadius=0
      boundBone {boneName=dyn_wick_01}
      boxExtents=<10 20 20>
      constraintSetup {angularLimitsMax=<20 0 15>, angularLimitsMin=<-20 -30 -15>, angularTarget=<0 0 0>}
      localJointOffset=<1 0 -1>
     }]
    relativeSpaceBone {boneName=spine_05}
   }
   n8 = Anim Dynamics - Bone: dyn_wick_02(ComponentPose=@n7, Alpha=1, ExternalForce="0,0,0") {
    angularDampingOverride=0.9999
    angularSpringConstant=50
    bAngularSpring=true
    bGravityOverrideInSimSpace=true
    bOverrideAngularDamping=true
    simulationSpace=RootRelative
    alphaBoolBlend {blendOption=Linear}
    boundBone {boneName=dyn_wick_02}
    componentAppliedLinearAccClamp=<3500 3500 3500>
    componentLinearAccScale=<3 3 3>
    componentLinearVelScale=<3 3 3>
    physicsBodyDefinitions [{
      sphereCollisionRadius=0
      boundBone {boneName=dyn_wick_02}
      boxExtents=<10 20 20>
      constraintSetup {angularLimitsMax=<20 0 15>, angularLimitsMin=<-20 -30 -15>, angularTarget=<0 0 0>}
      localJointOffset=<1 0 -1>
     }]
    relativeSpaceBone {boneName=spine_05}
   }
   n9 = Anim Dynamics - Bone: dyn_wick_03(ComponentPose=@n8, Alpha=1, ExternalForce="0,0,0") {
    angularDampingOverride=0.9999
    angularSpringConstant=50
    bAngularSpring=true
    bGravityOverrideInSimSpace=true
    bOverrideAngularDamping=true
    simulationSpace=RootRelative
    alphaBoolBlend {blendOption=Linear}
    boundBone {boneName=dyn_wick_03}
    componentAppliedLinearAccClamp=<3500 3500 3500>
    componentLinearAccScale=<3 3 3>
    componentLinearVelScale=<3 3 3>
    physicsBodyDefinitions [{
      sphereCollisionRadius=0
      boundBone {boneName=dyn_wick_03}
      boxExtents=<10 20 20>
      constraintSetup {angularLimitsMax=<20 0 15>, angularLimitsMin=<-20 -30 -15>, angularTarget=<0 0 0>}
      localJointOffset=<1 0 -1>
     }]
    relativeSpaceBone {boneName=spine_05}
   }
   Output Pose(Result=Component To Local(ComponentPose=@n9)) {layerGroup=None}
  ]}
 EventGraph {nodes [
   Event BlueprintUpdateAnimation()
   > n1 = Get Dynamic Material Instance(SkeletalMeshComponent=GetPartSkeletalMeshComponent())
   > n2 = Branch(Condition=Get bIsPlayingEmote())
   > SetScalarParameterValue(self=@n1, ParameterName=Smile Mouth animation Frame, Value=8)
   SetScalarParameterValue(execute=@n2.else, self=@n1, ParameterName=Smile Mouth animation Frame, Value=3)
  ]}
]
variables []
